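# Looping thrust sound cue for the Chainmail glider cosmetic.
# A mixer layers four looping waves (flap, gear, wind and tonal), with the gear layer mixed at half level.
# Each layer passes through a continuous modulator that scales pitch and volume from the AnyDirection parameter, so the loops rise as its value increases.
# Glider thrust attenuation, sound class and a cosmetic glider-loop properties sheet govern playback.
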SoundCue /BRCosmetics/Sounds/Gliders/Chainmail/GliderThrustLoop_Chainmail_Cue
root=Mixer_0
nodes [
 {
  id=Mixer_0
  children [ModulatorContinuous_0, ModulatorContinuous_1, ModulatorContinuous_2, ModulatorContinuous_3]
  properties {inputVolume [1,0.5,1,1]}
 }
 {
  id=ModulatorContinuous_0
  children [WavePlayer_0]
  properties {
   pitchModulationParams {default=0, maxOutput=1.1, minOutput=1, parameterName=AnyDirection}
   volumeModulationParams {default=0, minOutput=0.5, parameterName=AnyDirection}
  }
 }
 {
  id=ModulatorContinuous_1
  children [WavePlayer_1]
  properties {
   pitchModulationParams {default=0, maxOutput=1.1, minOutput=1, parameterName=AnyDirection}
   volumeModulationParams {default=0, maxOutput=0.8, minOutput=0.5, parameterName=AnyDirection}
  }
 }
 {
  id=ModulatorContinuous_2
  children [WavePlayer_2]
  properties {
   pitchModulationParams {default=0, maxOutput=1.2, minOutput=1, parameterName=AnyDirection}
   volumeModulationParams {default=0, minOutput=0.4, parameterName=AnyDirection}
  }
 }
 {
  id=ModulatorContinuous_3
  children [WavePlayer_3]
  properties {
   pitchModulationParams {default=0, maxOutput=1.2, minOutput=1, parameterName=AnyDirection}
   volumeModulationParams {default=0, maxOutput=1.2, minOutput=0.5, parameterName=AnyDirection}
  }
 }
 {id=WavePlayer_0, properties {bLooping=true, soundWaveAssetPtr=&/Glider_Chainmail_Thrust_Flap_LP}}
 {id=WavePlayer_1, properties {bLooping=true, soundWaveAssetPtr=&/Glider_Chainmail_Thrust_Gear_LP}}
 {id=WavePlayer_2, properties {bLooping=true, soundWaveAssetPtr=&/Glider_Chainmail_Thrust_Wind_LP}}
 {id=WavePlayer_3, properties {bLooping=true, soundWaveAssetPtr=&/Glider_Chainmail_Thrust_Tonal_LP}}
]
properties {
 attenuationSettings=&Fort_Player_Glider_Thrust_Athena_Att
 audioPropertiesSheet=&APS_Cosmetic_Glider_Loop
 maxDistance=12500
 soundClassObject=&Fort_Gliders_Land
 volumeMultiplier=0.7
 concurrencySet [null]
}
